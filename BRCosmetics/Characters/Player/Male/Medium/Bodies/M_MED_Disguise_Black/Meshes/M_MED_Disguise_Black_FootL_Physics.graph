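# Physics asset for the left foot mesh of the male medium Disguise Black body cosmetic, covering shoelace and tassel bones.
# Small capsule bodies form two three-link shoelace chains from a shared master bone, plus a two-link tassel chain from its own master, named for the right side.
# Constraints use tight swing and twist limits, except the second lace chain's root joint, which swings and twists freely.
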PhysicsAsset /BRCosmetics/Characters/Player/Male/Medium/Bodies/M_MED_Disguise_Black/Meshes/M_MED_Disguise_Black_FootL_Physics
bodies {
 dyn_shoelace_01_l [{shapeType=Capsule, length=0.5, radius=0.5, shapeName=dyn_shoelace_01_l_capsule, center=<0 0 0.3873>}]
 dyn_shoelace_02_l [{shapeType=Capsule, length=0.5, radius=0.5, shapeName=dyn_shoelace_02_l_capsule, center=<0 0 0.3873>}]
 dyn_shoelace_03_l [{shapeType=Capsule, length=0.5, radius=0.5, shapeName=dyn_shoelace_03_l_capsule, center=<0 0 0.3873>}]
 dyn_shoelace_04_l [{shapeType=Capsule, length=0.5, radius=0.5, shapeName=dyn_shoelace_04_l_capsule}]
 dyn_shoelace_05_l [{shapeType=Capsule, length=0.5, radius=0.5, shapeName=dyn_shoelace_05_l_capsule, center=<0 0 0.3873>}]
 dyn_shoelace_06_l [{shapeType=Capsule, length=0.5, radius=0.5, shapeName=dyn_shoelace_06_l_capsule, center=<0 0 0.3873>}]
 dyn_shoelace_master_l [{shapeType=Capsule, length=0.5, radius=0.5, shapeName=dyn_shoelace_master_l_capsule}]
 dyn_tassle_01_r [{shapeType=Capsule, length=1, radius=1, shapeName=dyn_tassle_01_r_capsule, center=<1.5021 0 0>, rotation=<rot 90 0 0>}]
 dyn_tassle_02_r [{shapeType=Capsule, length=1, radius=1, shapeName=dyn_shoelace_06_l_capsule, center=<1.5021 0 0>, rotation=<rot 90 0 0>}]
 dyn_tassle_master_r [{shapeType=Capsule, length=0.1, radius=0.1, shapeName=dyn_tassle_master_r_capsule, rotation=<rot 90 0 0>}]
}
constraints [
 {bone1Name=dyn_shoelace_01_l, bone2Name=dyn_shoelace_master_l, swing1LimitDegrees=8, swing2LimitDegrees=8, twistLimitDegrees=5}
 {bone1Name=dyn_shoelace_02_l, bone2Name=dyn_shoelace_01_l, swing1LimitDegrees=8, swing2LimitDegrees=8, twistLimitDegrees=5}
 {bone1Name=dyn_shoelace_03_l, bone2Name=dyn_shoelace_02_l, swing1LimitDegrees=8, swing2LimitDegrees=8, twistLimitDegrees=5}
 {
  bone1Name=dyn_shoelace_04_l
  bone2Name=dyn_shoelace_master_l
  swing1LimitDegrees=45
  swing1Motion=Free
  swing2LimitDegrees=45
  swing2Motion=Free
  twistLimitDegrees=45
  twistMotion=Free
 }
 {bone1Name=dyn_shoelace_05_l, bone2Name=dyn_shoelace_04_l, swing1LimitDegrees=8, swing2LimitDegrees=8, twistLimitDegrees=5}
 {bone1Name=dyn_shoelace_06_l, bone2Name=dyn_shoelace_05_l, swing1LimitDegrees=8, swing2LimitDegrees=8, twistLimitDegrees=5}
 {bone1Name=dyn_tassle_01_r, bone2Name=dyn_tassle_master_r, swing1LimitDegrees=5, swing2LimitDegrees=5, twistLimitDegrees=5}
 {bone1Name=dyn_tassle_02_r, bone2Name=dyn_tassle_01_r, swing1LimitDegrees=15, swing2LimitDegrees=15, twistLimitDegrees=5}
]
