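# Physics asset for the right-foot mesh of the male medium Disguise Black player body, providing small capsule bodies for dangling shoelace and tassel bones.
# Two three-bone shoelace strands branch from a right-side master, and a two-bone tassel chain, named as left-side, hangs from its own master.
# Joints have tight swing and twist limits, except the second strand's root (shoelace 04), whose angular motion is free.
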PhysicsAsset /BRCosmetics/Characters/Player/Male/Medium/Bodies/M_MED_Disguise_Black/Meshes/M_MED_Disguise_Black_FootR_Physics
bodies {
 dyn_shoelace_01_r [{shapeType=Capsule, length=0.5, radius=0.5, shapeName=dyn_shoelace_01_r_capsule, center=<0.7747 0 0>, rotation=<rot 90 0 0>}]
 dyn_shoelace_02_r [{shapeType=Capsule, length=0.5, radius=0.5, shapeName=dyn_shoelace_02_r_capsule, center=<0.7747 0 0>, rotation=<rot 90 0 0>}]
 dyn_shoelace_03_r [{shapeType=Capsule, length=0.5, radius=0.5, shapeName=dyn_shoelace_03_r_capsule, center=<0.7747 0 0>, rotation=<rot 90 0 0>}]
 dyn_shoelace_04_r [{shapeType=Capsule, length=0.5, radius=0.5, shapeName=dyn_shoelace_04_r_capsule}]
 dyn_shoelace_05_r [{shapeType=Capsule, length=0.5, radius=0.5, shapeName=dyn_shoelace_05_r_capsule, center=<0.7747 0 0>, rotation=<rot 90 0 0>}]
 dyn_shoelace_06_r [{shapeType=Capsule, length=0.5, radius=0.5, shapeName=dyn_shoelace_06_r_capsule, center=<0.7747 0 0>, rotation=<rot 90 0 0>}]
 dyn_shoelace_master_r [{shapeType=Capsule, length=0.5, radius=0.5, shapeName=dyn_shoelace_master_r_capsule}]
 dyn_tassle_01_l [{shapeType=Capsule, length=0.5, radius=1, shapeName=dyn_tassle_01_l_capsule, center=<0.7747 0 0>, rotation=<rot 90 0 0>}]
 dyn_tassle_02_l [{shapeType=Capsule, length=0.5, radius=1, shapeName=dyn_tassle_02_l_capsule, center=<0.7747 0 0>, rotation=<rot 90 0 0>}]
 dyn_tassle_master_l [{shapeType=Capsule, length=0.5, radius=0.5, shapeName=dyn_tassle_master_l_capsule}]
}
constraints [
 {bone1Name=dyn_shoelace_01_r, bone2Name=dyn_shoelace_master_r, swing1LimitDegrees=8, swing2LimitDegrees=8, twistLimitDegrees=5}
 {bone1Name=dyn_shoelace_02_r, bone2Name=dyn_shoelace_01_r, swing1LimitDegrees=8, swing2LimitDegrees=8, twistLimitDegrees=5}
 {bone1Name=dyn_shoelace_03_r, bone2Name=dyn_shoelace_02_r, swing1LimitDegrees=8, swing2LimitDegrees=8, twistLimitDegrees=5}
 {
  bone1Name=dyn_shoelace_04_r
  bone2Name=dyn_shoelace_master_r
  swing1LimitDegrees=45
  swing1Motion=Free
  swing2LimitDegrees=45
  swing2Motion=Free
  twistLimitDegrees=45
  twistMotion=Free
 }
 {bone1Name=dyn_shoelace_05_r, bone2Name=dyn_shoelace_04_r, swing1LimitDegrees=8, swing2LimitDegrees=8, twistLimitDegrees=5}
 {bone1Name=dyn_shoelace_06_r, bone2Name=dyn_shoelace_05_r, swing1LimitDegrees=8, swing2LimitDegrees=8, twistLimitDegrees=5}
 {bone1Name=dyn_tassle_01_l, bone2Name=dyn_tassle_master_l, swing1LimitDegrees=5, swing2LimitDegrees=5, twistLimitDegrees=5}
 {bone1Name=dyn_tassle_02_l, bone2Name=dyn_tassle_01_l, swing1LimitDegrees=8, swing2LimitDegrees=8, twistLimitDegrees=5}
]
