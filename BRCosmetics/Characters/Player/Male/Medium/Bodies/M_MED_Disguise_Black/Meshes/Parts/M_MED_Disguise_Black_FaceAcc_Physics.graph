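# Physics asset for the face accessory part of the male medium black Disguise cosmetic body, giving physics bodies to the dynamic hoodie bones around the head.
# Five hood chains (left, midleft, mid, midright, right) each run from a master bone through two segments, all built from capsules alongside one large head capsule.
# The segments have tight swing and twist limits, while each master joins the head with free angular motion.
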PhysicsAsset /BRCosmetics/Characters/Player/Male/Medium/Bodies/M_MED_Disguise_Black/Meshes/Parts/M_MED_Disguise_Black_FaceAcc_Physics
bodies {
 dyn_hoodie_01_left [{shapeType=Capsule, length=1.7978, radius=2.5607, shapeName=dyn_hoodie_01_left_capsule, center=<3.7911 -0.4655 0.6845>, rotation=<rot 90 0 0>}]
 dyn_hoodie_01_mid [{shapeType=Capsule, length=2.9388, radius=2.5357, shapeName=dyn_hoodie_01_mid_capsule, center=<4.0143 -0.8446 1.5639>, rotation=<rot 90 0 0>}]
 dyn_hoodie_01_midleft [{
   shapeType=Capsule
   length=3.3441
   radius=2.5968
   shapeName=dyn_hoodie_01_midleft_capsule
   center=<4.201 -0.3759 -1.6454>
   rotation=<rot 90 0 0>
  }]
 dyn_hoodie_01_midright [{
   shapeType=Capsule
   length=3.1158
   radius=2.5946
   shapeName=dyn_hoodie_01_midright_capsule
   center=<4.0875 -1.6317 0.5375>
   rotation=<rot 90 0 0>
  }]
 dyn_hoodie_01_right [{shapeType=Capsule, length=1.9314, radius=3.4175, shapeName=dyn_hoodie_01_right_capsule, center=<2.4681 -0.173 0.0154>, rotation=<rot 90 0 0>}]
 dyn_hoodie_02_left [{shapeType=Capsule, length=1, radius=2, shapeName=dyn_hoodie_02_left_capsule, center=<1.1181 0 -0.0001>, rotation=<rot 90 0 0>}]
 dyn_hoodie_02_mid [{shapeType=Capsule, length=1, radius=2, shapeName=dyn_hoodie_02_mid_capsule, center=<1.1181 0 -0.0001>, rotation=<rot 90 0 0>}]
 dyn_hoodie_02_midleft [{shapeType=Capsule, length=1, radius=2, shapeName=dyn_hoodie_02_midleft_capsule, center=<1.1181 -1.2593 -0.7672>, rotation=<rot 90 0 0>}]
 dyn_hoodie_02_midright [{shapeType=Capsule, length=1, radius=2, shapeName=dyn_hoodie_02_midright_capsule, center=<1.1181 -1.4813 0.4542>, rotation=<rot 90 0 0>}]
 dyn_hoodie_02_right [{shapeType=Capsule, length=1, radius=2, shapeName=dyn_hoodie_02_right_capsule, center=<1.1181 0 -0.0001>, rotation=<rot 90 0 0>}]
 dyn_hoodie_master_left [{shapeType=Capsule, length=0.505, radius=0.505, shapeName=dyn_hoodie_master_left_capsule, rotation=<rot 0 -90 0>}]
 dyn_hoodie_master_mid [{shapeType=Capsule, length=0.505, radius=0.505, shapeName=dyn_hoodie_master_mid_capsule, rotation=<rot 0 -90 0>}]
 dyn_hoodie_master_midleft [{shapeType=Capsule, length=0.505, radius=0.505, shapeName=dyn_hoodie_master_midleft_capsule, rotation=<rot 0 -90 0>}]
 dyn_hoodie_master_midright [{shapeType=Capsule, length=0.505, radius=0.505, shapeName=dyn_hoodie_master_midright_capsule, rotation=<rot 0 -90 0>}]
 dyn_hoodie_master_right [{shapeType=Capsule, length=0.505, radius=0.505, shapeName=dyn_hoodie_master_right_capsule, rotation=<rot 0 -90 0>}]
 head [{shapeType=Capsule, length=3.25, radius=7.5, shapeName=head_capsule, center=<0 8.4112 0.0001>, rotation=<rot 90 0 -28.5428>}]
}
constraints [
 {bone1Name=dyn_hoodie_01_left, bone2Name=dyn_hoodie_master_left, swing1LimitDegrees=15, swing2LimitDegrees=15, twistLimitDegrees=5}
 {bone1Name=dyn_hoodie_02_left, bone2Name=dyn_hoodie_01_left, swing1LimitDegrees=15, swing2LimitDegrees=15, twistLimitDegrees=5}
 {bone1Name=dyn_hoodie_01_midleft, bone2Name=dyn_hoodie_master_midleft, swing1LimitDegrees=8, swing2LimitDegrees=8, twistLimitDegrees=5}
 {bone1Name=dyn_hoodie_02_midleft, bone2Name=dyn_hoodie_01_midleft, swing1LimitDegrees=15, swing2LimitDegrees=15, twistLimitDegrees=5}
 {bone1Name=dyn_hoodie_01_mid, bone2Name=dyn_hoodie_master_mid, swing1LimitDegrees=15, swing2LimitDegrees=8, twistLimitDegrees=5}
 {bone1Name=dyn_hoodie_01_midright, bone2Name=dyn_hoodie_master_midright, swing1LimitDegrees=8, swing2LimitDegrees=8, twistLimitDegrees=5}
 {bone1Name=dyn_hoodie_02_midright, bone2Name=dyn_hoodie_01_midright, swing1LimitDegrees=15, swing2LimitDegrees=15, twistLimitDegrees=5}
 {bone1Name=dyn_hoodie_01_right, bone2Name=dyn_hoodie_master_right, swing1LimitDegrees=15, swing2LimitDegrees=15, twistLimitDegrees=5}
 {bone1Name=dyn_hoodie_02_right, bone2Name=dyn_hoodie_01_right, swing1LimitDegrees=15, swing2LimitDegrees=15, twistLimitDegrees=5}
 {bone1Name=dyn_hoodie_02_mid, bone2Name=dyn_hoodie_01_mid, swing1LimitDegrees=15, swing2LimitDegrees=8, twistLimitDegrees=5}
 {
  bone1Name=dyn_hoodie_master_left
  bone2Name=head
  swing1LimitDegrees=45
  swing1Motion=Free
  swing2LimitDegrees=45
  swing2Motion=Free
  twistLimitDegrees=45
  twistMotion=Free
 }
 {
  bone1Name=dyn_hoodie_master_midleft
  bone2Name=head
  swing1LimitDegrees=45
  swing1Motion=Free
  swing2LimitDegrees=45
  swing2Motion=Free
  twistLimitDegrees=45
  twistMotion=Free
 }
 {
  bone1Name=dyn_hoodie_master_mid
  bone2Name=head
  swing1LimitDegrees=45
  swing1Motion=Free
  swing2LimitDegrees=45
  swing2Motion=Free
  twistLimitDegrees=45
  twistMotion=Free
 }
 {
  bone1Name=dyn_hoodie_master_midright
  bone2Name=head
  swing1LimitDegrees=45
  swing1Motion=Free
  swing2LimitDegrees=45
  swing2Motion=Free
  twistLimitDegrees=45
  twistMotion=Free
 }
 {
  bone1Name=dyn_hoodie_master_right
  bone2Name=head
  swing1LimitDegrees=45
  swing1Motion=Free
  swing2LimitDegrees=45
  swing2Motion=Free
  twistLimitDegrees=45
  twistMotion=Free
 }
]
