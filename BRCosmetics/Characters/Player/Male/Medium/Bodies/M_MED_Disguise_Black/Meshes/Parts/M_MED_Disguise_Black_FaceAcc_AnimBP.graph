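# Post-processes the face accessory mesh of the male medium black disguise body, simulating and reshaping its hoodie bones.
# The live chain copies the attached parent's pose and runs RigidBody physics in spine_05 space using a Thin physics asset.
# Fixed additive bone offsets follow the simulation, adjusting the hoodie's middle and sides.
# An alternate offsets-before-physics chain and a spare RigidBody node are disconnected from the output; no event logic exists.
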AnimBlueprint /BRCosmetics/Characters/Player/Male/Medium/Bodies/M_MED_Disguise_Black/Meshes/Parts/M_MED_Disguise_Black_FaceAcc_AnimBP
target_skeleton=&/M_MED_Disguise_Black_FaceAcc_Skeleton
graphs [
 AnimGraph {nodes [
   n1 = "Transform (Modify) Bone - Bone: dyn_hoodie_master"(
    ComponentPose=Local To Component(LocalPose=Copy Pose From Mesh() {bUseAttachedParent=true})
    Alpha=1
    Translation="0.321,0.000,2.217"
    Rotation="0.000,0.000,-5"
    Scale="1,1,1"
   ) {
    rotationMode=BMM_Additive
    scaleMode=BMM_Replace
    scaleSpace=BCS_BoneSpace
    translationMode=BMM_Additive
    alphaBoolBlend {blendOption=Linear}
    boneToModify {boneName=dyn_hoodie_master}
    rotation {roll=-5}
    translation {x=0.321, z=2.217}
   }
   n2 = "Transform (Modify) Bone - Bone: dyn_hoodie_master_midleft"(ComponentPose=@n1, Alpha=1, Translation="0,0,0", Rotation="0.000,-10,0.000", Scale="1,1,1") {
    rotationMode=BMM_Additive
    translationMode=BMM_Additive
    alphaBoolBlend {blendOption=Linear}
    boneToModify {boneName=dyn_hoodie_master_midleft}
    rotation {yaw=-10}
   }
   n3 = "Transform (Modify) Bone - Bone: dyn_hoodie_master_midright"(ComponentPose=@n2, Alpha=1, Translation="-1.259,0.000,0.000", Rotation="0.000,10,0", Scale="1,1,1") {
    rotationMode=BMM_Additive
    translationMode=BMM_Additive
    alphaBoolBlend {blendOption=Linear}
    boneToModify {boneName=dyn_hoodie_master_midright}
    rotation {yaw=10}
    translation {x=-1.259}
   }
   n4 = "Transform (Modify) Bone - Bone: dyn_hoodie_master_left"(ComponentPose=@n3, Alpha=1, Translation="0.771,0.000,0.000", Rotation="0.000,-25,0.000", Scale="1,1,1") {
    rotationMode=BMM_Additive
    translationMode=BMM_Additive
    alphaBoolBlend {blendOption=Linear}
    boneToModify {boneName=dyn_hoodie_master_left}
    rotation {yaw=-25}
    translation {x=0.771}
   }
   n5 = "Transform (Modify) Bone - Bone: dyn_hoodie_master_right"(ComponentPose=@n4, Alpha=1, Translation="-1.259,0.000,0.000", Rotation="0.000,10,0", Scale="1,1,1") {
    rotationMode=BMM_Additive
    translationMode=BMM_Additive
    alphaBoolBlend {blendOption=Linear}
    boneToModify {boneName=dyn_hoodie_master_right}
    rotation {yaw=10}
    translation {x=-1.259}
   }
   n6 = "Transform (Modify) Bone - Bone: dyn_hoodie_master_botleft"(ComponentPose=@n5, Alpha=1, Translation="0,0,0", Rotation="-15.557,-14.181,3.876", Scale="1,1,1") {
    rotationMode=BMM_Additive
    translationMode=BMM_Additive
    alphaBoolBlend {blendOption=Linear}
    boneToModify {boneName=dyn_hoodie_master_botleft}
    rotation=<rot -15.557 -14.1811 3.876>
   }
   n7 = RigidBody(ComponentPose=@n6) {
    bOverrideWorldGravity=true
    overridePhysicsAsset=&/M_MED_Disguise_Black_FaceAcc_Thin_Phys
    simulationSpace=BaseBoneSpace
    alphaBoolBlend {blendOption=Linear}
    baseBoneRef {boneName=spine_05}
    componentAppliedLinearAccClamp=<2000 2000 2000>
    componentLinearAccScale=<0.3 0.3 0.35>
    componentLinearVelScale=<0.3 0.3 0.35>
    overrideWorldGravity {z=-980}
   }
   Component To Local(ComponentPose=@n7)
   RigidBody() {
    bOverrideWorldGravity=true
    overridePhysicsAsset=&/M_MED_Disguise_Black_FaceAcc_Physics
    simulationSpace=BaseBoneSpace
    alphaBoolBlend {blendOption=Linear}
    baseBoneRef {boneName=spine_05}
    componentAppliedLinearAccClamp=<2000 2000 2000>
    componentLinearAccScale=<0.3 0.3 0.35>
    componentLinearVelScale=<0.3 0.3 0.35>
    overrideWorldGravity {z=-980}
   }
   n8 = "Transform (Modify) Bone - Bone: dyn_hoodie_master"(
    ComponentPose=Local To Component(LocalPose=Copy Pose From Mesh() {bUseAttachedParent=true})
    Alpha=1
    Translation="0,0,0"
    Rotation="0,0,0"
    Scale="1,1,1"
   ) {
    rotationMode=BMM_Additive
    scaleMode=BMM_Replace
    scaleSpace=BCS_BoneSpace
    translationMode=BMM_Additive
    alphaBoolBlend {blendOption=Linear}
    boneToModify {boneName=dyn_hoodie_master}
   }
   n9 = RigidBody(ComponentPose=@n8) {
    bOverrideWorldGravity=true
    overridePhysicsAsset=&M_MED_Disguise_Black_FaceAcc_Thin_Phys
    simulationSpace=BaseBoneSpace
    alphaBoolBlend {blendOption=Linear}
    baseBoneRef {boneName=spine_05}
    componentAppliedLinearAccClamp=<2000 2000 2000>
    componentLinearAccScale=<0.3 0.3 0.35>
    componentLinearVelScale=<0.3 0.3 0.35>
    overrideWorldGravity {z=-980}
   }
   n10 = "Transform (Modify) Bone - Bone: dyn_hoodie_master_mid"(ComponentPose=@n9, Alpha=1, Translation="0.000,0.093,1.667", Rotation="0.000,0.000,-7.774", Scale="1,1,1") {
    rotationMode=BMM_Additive
    translationMode=BMM_Additive
    alphaBoolBlend {blendOption=Linear}
    boneToModify {boneName=dyn_hoodie_master_mid}
    rotation {roll=-7.774}
    translation {y=0.093, z=1.667}
   }
   n11 = "Transform (Modify) Bone - Bone: dyn_hoodie_master_midleft"(ComponentPose=@n10, Alpha=1, Translation="0,0,0", Rotation="0,-12.636,0.000", Scale="1,1,1") {
    rotationMode=BMM_Additive
    translationMode=BMM_Additive
    alphaBoolBlend {blendOption=Linear}
    boneToModify {boneName=dyn_hoodie_master_midleft}
    rotation {yaw=-12.636}
   }
   n12 = "Transform (Modify) Bone - Bone: dyn_hoodie_master_midright"(ComponentPose=@n11, Alpha=1, Translation="0,0,0", Rotation="0.000,12.644,0.000", Scale="1,1,1") {
    rotationMode=BMM_Additive
    translationMode=BMM_Additive
    alphaBoolBlend {blendOption=Linear}
    boneToModify {boneName=dyn_hoodie_master_midright}
    rotation {yaw=12.644}
   }
   n13 = "Transform (Modify) Bone - Bone: dyn_hoodie_master_left"(ComponentPose=@n12, Alpha=1, Translation="0,0,0", Rotation="5.835,-18.559,-17.228", Scale="1,1,1") {
    rotationMode=BMM_Additive
    translationMode=BMM_Additive
    alphaBoolBlend {blendOption=Linear}
    boneToModify {boneName=dyn_hoodie_master_left}
    rotation=<rot 5.835 -18.559 -17.228>
   }
   n14 = "Transform (Modify) Bone - Bone: dyn_hoodie_master_right"(ComponentPose=@n13, Alpha=1, Translation="-1.021,0.000,0.000", Rotation="0.000,12.644,0.000", Scale="1,1,1") {
    rotationMode=BMM_Additive
    translationMode=BMM_Additive
    alphaBoolBlend {blendOption=Linear}
    boneToModify {boneName=dyn_hoodie_master_right}
    rotation {yaw=12.644}
    translation {x=-1.021}
   }
   n15 = "Transform (Modify) Bone - Bone: dyn_hoodie_master_midright"(ComponentPose=@n14, Alpha=1, Translation="-3.687,0.000,1.442", Rotation="0.000,-37.930,0.000", Scale="1,1,1") {
    rotationMode=BMM_Additive
    translationMode=BMM_Additive
    alphaBoolBlend {blendOption=Linear}
    boneToModify {boneName=dyn_hoodie_master_midright}
    rotation {yaw=-37.9299}
    translation {x=-3.687, z=1.442}
   }
   n16 = "Transform (Modify) Bone - Bone: dyn_hoodie_master_right"(ComponentPose=@n15, Alpha=1, Translation="-1.163,0.000,0.000", Rotation="0.000,-22.496,0.000", Scale="1,1,1") {
    rotationMode=BMM_Additive
    translationMode=BMM_Additive
    alphaBoolBlend {blendOption=Linear}
    boneToModify {boneName=dyn_hoodie_master_right}
    rotation {yaw=-22.4958}
    translation {x=-1.163}
   }
   Output Pose(Result=Component To Local(ComponentPose=@n16)) {layerGroup=None}
  ]}
 EventGraph {nodes [Event BlueprintUpdateAnimation()]}
]
variables []
